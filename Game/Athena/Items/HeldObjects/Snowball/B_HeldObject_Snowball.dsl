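; Snowball held object that players pick up and throw, exploding on impact.
; Impact after a throw, entering water unheld, or losing all health queues an async sphere overlap, then plays death effects and destroys the actor.
; Line-of-sight-filtered targets get Curie water and team-filtered damage, and pawns also get low friction and a launch.
; Held-object events track state and toggle surface linking, with a cosmetic-side reattach check fixing bad attachment.
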(event Custom|Play%20Drop%20Animation
  (bind (_returnvalue_x _returnvalue_y _returnvalue_z) (Math|Vector|BreakVector (Transformation|GetActorLocation)))
  (bind (_returnvalue_x_2 _returnvalue_y_2) (Math|Vector|BreakVector (Transformation|GetActorLocation)))
  ;[1 1 1] Drop Animation
  (Variables|Default|SetStartZPos _returnvalue_z)
  (bind (_direction _z_pos) (Timeline|DropAnimation Play)
    (:Update
      (Transformation|SetActorLocation 0 (Math|Vector|MakeVector _returnvalue_x_2 _returnvalue_y_2 (+ (* _z_pos 9e+01) (Variables|Default|GetStartZPos))) false true)
      (Variables|/Script/Engine.MovementComponent|SetVelocity "0, 0, 0" (Variables|Default|GetFortProjectileMovement)))
    (:Finished
      (Variables|Default|SetDoneDropping true))))

(event Custom|StopDropping
  (bind (_returnvalue_x _returnvalue_y) (Math|Vector|BreakVector (Transformation|GetActorLocation)))
  ;[1 1 1] Drop Animation
  (Variables|Default|SetDoneDropping true)
  (Timeline|DropAnimation Stop))

(event Custom|EnterWater
  (params (WaterBody :type Fort%20Water%20Body%20Actor%20Object%20Reference) (WaterInteractionComponent :type Fort%20Water%20Interaction%20Component%20Object%20Reference) (bIsFirstBody :type Boolean))
  ;[1 1 1] Explode in water
  (if (not (Variables|Default|GetHeld))
    (CallFunction|BlowUp (Variables|Default|GetLastHolder))))

(event EventBeginPlay
  (|CallParentFunction)
  (Utilities|FlowControl|Sequence
    (:then_1
      (Collision|SetCollisionResponsetoChannel (Variables|Default|GetStaticMesh) "ECC_Camera" "ECR_Ignore")
      ;[1 1 1] Explode in water
      (Water|BindEventtoOnEnterWaterEvent (Variables|Default|GetFortWaterInteraction) (delegate Custom|EnterWater)))
    (:then_2
      (EventDispatchers|BindEventtoOnDamaged (EventDispatchers|CreateEvent :!event Custom|OnHit) :!class /Script/FortniteGame.BuildingActor)
      (EventDispatchers|BindEventtoOnDied (EventDispatchers|CreateEvent :!event Custom|EventSnowballDeath) :!class /Script/FortniteGame.BuildingActor))))

(event Custom|EventSnowballDeath
  (params (DamagedActor :type Actor%20Object%20Reference) (Damage :type Float%20%28single-precision%29) (InstigatedBy :type Controller%20Object%20Reference) (DamageCauser :type Actor%20Object%20Reference) (HitLocation :type Vector) (FHitComponent :type Primitive%20Component%20Object%20Reference) (BoneName :type Name) (Momentum :type Vector))
  ;[1 1 1] Death effects
  (Development|PrintString "DEATH" true true "(R=0.000000,G=0.660000,B=1.000000,A=1.000000)" 1e+01)
  (Utilities|FlowControl|Sequence
    (:then_0
      (Rendering|SetVisibility (Variables|Default|GetStaticMesh) false false)
      (Collision|SetActorEnableCollision false))
    (:then_1
      (Variables|Default|SetDied true)
      (Effects|SpawnEmitteratLocation :EmitterTemplate (select (Water|IsinWater (Variables|Default|GetFortWaterInteraction)) "/Game/Athena/Items/Consumables/SneakySnowman/FX/P_SnowBaller_EnterWater_V2.P_SnowBaller_EnterWater_V2" "/Game/Athena/Items/Consumables/SneakySnowman/FX/P_SnowBaller_Explosion_V2.P_SnowBaller_Explosion_V2") :Location (Transformation|GetActorLocation) :Scale (/ (Transformation|GetWorldScale (Variables|Default|GetStaticMesh)) 2.0) :bAutoDestroy true :PoolingMethod "None" :bAutoActivateSystem true)
      (Audio|PlaySoundatLocation "/Game/Sounds/Fort_Gadgets/SneakySnowman/SneakySnowman_BallDestroy_Cue.SneakySnowman_BallDestroy_Cue" (Transformation|GetActorLocation) "0, 0, 0" 1.5 0.8 0.0 0 0 self)
      (Actor|DestroyActor))))

(event OnHeldObjectDropped(FortHeldObject)
  ;[1 1 1] Set variables to track player interaction
  (Variables|Default|SetHeld false))

(event EventOnDeathServer
  (params (Damage :type Float%20%28single-precision%29) (DamageTags :type Gameplay%20Tag%20Container%20Structure%20%28by%20ref%29) (Momentum :type Vector :default "0, 0, 0") (HitInfo :type Hit%20Result%20Structure%20%28by%20ref%29) (InstigatedBy :type Controller%20Object%20Reference) (DamageCauser :type Actor%20Object%20Reference) (EffectContext :type Gameplay%20Effect%20Context%20Handle%20Structure))
  ;[1 1 1] Trigger blow up when it dies (runs out of health)
  (Utilities|IsValid InstigatedBy
    (:Is%20Valid
      (bind _asfort_player_pawn (Utilities|Casting|CastToFortPlayerPawn (Pawn|GetControlledPawn InstigatedBy))
        (:then
          _)
        (:CastFailed
          _))
      (CallFunction|BlowUp _asfort_player_pawn))))

(event OnHeldObjectPlaced(FortHeldObject)
  ;[1 1 1] Set variables to track player interaction
  (Variables|Default|SetHeld false))

(event OnHeldObjectPickedUp(FortHeldObject)
  ;[1 1 1] Set variables to track player interaction
  (Variables|Default|SetHeld true))

(event OnHeldObjectThrown(FortHeldObject)
  ;[1 1 1] Set variables to track player interaction
  (Variables|Default|SetWasJustThrown true)
  (Variables|Default|SetWasThrown true)
  (Variables|Default|SetHeld false))

(event Interaction|EventBlueprintOnInteract
  (params (InteractingPawn :type Fort%20Pawn%20Object%20Reference) (InteractionBeingAttempted :type EInteractionBeingAttempted%20Enum :default "FirstInteraction"))
  ;[1 1 1] On Interact
  (switch Utilities|FlowControl|Switch|SwitchonEInteractionBeingAttempted InteractionBeingAttempted
    (:SecondInteraction
      (bind _asfort_player_pawn (Utilities|Casting|CastToFortPlayerPawn InteractingPawn))
      (CallFunction|SetInstigator _asfort_player_pawn)
      (Variables|/Script/Engine.MovementComponent|SetVelocity "0, 0, 0" (Variables|Default|GetFortProjectileMovement))
      (if (Networking|IsServer)
        (HeldObjects|Behavior|PickupHeldObject (Variables|Default|GetFortHeldObject) _asfort_player_pawn)))))

(event Custom|SetInstigator
  (params (Instigator :type Fort%20Player%20Pawn%20Object%20Reference))
  ;[1 1 1] On Interact
  (Variables|Default|SetLastHolder Instigator)
  (Variables|Actor|SetInstigator Instigator))

(event Custom|OnHit
  (params (DamagedActor :type Actor%20Object%20Reference) (Damage :type Float%20%28single-precision%29) (InstigatedBy :type Controller%20Object%20Reference) (DamageCauser :type Actor%20Object%20Reference) (HitLocation :type Vector) (FHitComponent :type Primitive%20Component%20Object%20Reference) (BoneName :type Name) (Momentum :type Vector))
  ;[1 1 1] Hit Effects
  (Effects|SpawnEmitteratLocation "/Game/Athena/Items/Consumables/SneakySnowman/FX/P_Snowball_WorldImpact.P_Snowball_WorldImpact" HitLocation (Math|Conversions|RotationFromXVector (- (Transformation|GetActorLocation DamageCauser) HitLocation)) (/ (Transformation|GetWorldScale (Variables|Default|GetStaticMesh)) 1e+01))
  (Audio|PlaySoundatLocation "/Game/Sounds/Fort_Gadgets/SneakySnowman/SneakySnowman_Pickaxe_Impact_Cue.SneakySnowman_Pickaxe_Impact_Cue" HitLocation))

(event Custom|BlowUp
  (params (LastHolder :type Fort%20Player%20Pawn%20Object%20Reference))
  (bind _explosionradius (Variables|Default|GetExplosionRadius))
  (Variables|Default|SetLastHolder LastHolder)
  (CallFunction|StopDropping)
  (if (not (Variables|Default|GetDied))
    (Utilities|IsValid (Variables|Default|GetLastHolder)
      (:Is%20Valid
        ;[1 1 1] Queue Damage, Frost feet, and impulse
        (Utilities|FlowControl|DoOnce false
          (:Completed
            (Class|AsyncOverlapQueue|AsyncSphereOverlapActors :Location (Transformation|GetActorLocation) :Radius _explosionradius :ShouldCancelProcess (EventDispatchers|CreateEvent :!event OnDamageCancelled) :ProcessHitActorFunction (EventDispatchers|CreateEvent :!event Custom|DoDamageEffectTotarget) :CompletionFunction (EventDispatchers|CreateEvent :!event Custom|OnDamageCompleted) :TraceChannel "ECC_GameTraceChannel5")
            ;[1 1 1] Apply Curie water to landscape
            (Curie|ExecuteWaterApplication (Variables|Default|GetLastHolder) (Variables|Default|GetLastHolder) (.HitActor (Collision|BreakHitResult (stale OnComponentHit(Capsule) Hit))) 1e+02 (stale OnComponentHit(Capsule) Hit) _explosionradius))) ; Don't hit multiple times
      ))))

(event OnComponentHit(Capsule)
  (params (HitComponent :type Primitive%20Component%20Object%20Reference) (OtherActor :type Actor%20Object%20Reference) (OtherComp :type Primitive%20Component%20Object%20Reference) (NormalImpulse :type Vector :default "0, 0, 0") (Hit :type Hit%20Result%20Structure%20%28by%20ref%29))
  (bind (_bblockinghit _binitialoverlap _time _distance _location _impactpoint _normal _impactnormal) (Collision|BreakHitResult Hit))
  (bind (_bblockinghit_2 _binitialoverlap_2 _time_2 _distance_2 _location_2 _impactpoint_2 _normal_2 _impactnormal_2 _physmat _hitactor) (Collision|BreakHitResult Hit))
  (bind _explosionradius (Variables|Default|GetExplosionRadius))
  (Variables|Default|SetImpactNormal _impactnormal)
  (if (Variables|Default|GetWasThrown)
    (if (not (Variables|Default|GetDied))
      (Utilities|IsValid (Variables|Default|GetLastHolder)
        (:Is%20Valid
          ;[1 1 1] Queue Damage, Frost feet, and impulse
          (Utilities|FlowControl|DoOnce false
            (:Completed
              (Class|AsyncOverlapQueue|AsyncSphereOverlapActors :Location (Transformation|GetActorLocation) :Radius _explosionradius :ShouldCancelProcess (EventDispatchers|CreateEvent :!event OnDamageCancelled) :ProcessHitActorFunction (EventDispatchers|CreateEvent :!event Custom|DoDamageEffectTotarget) :CompletionFunction (EventDispatchers|CreateEvent :!event Custom|OnDamageCompleted) :TraceChannel "ECC_GameTraceChannel5")
              ;[1 1 1] Apply Curie water to landscape
              (Curie|ExecuteWaterApplication (Variables|Default|GetLastHolder) (Variables|Default|GetLastHolder) _hitactor 1e+02 Hit _explosionradius))) ; Don't hit multiple times
        )))))

(event OnLinkedActorDisconnected(FortLinkToActor)
  ;[1 1 1] Note: This needs to move to the parent post 11.31 but it changed in 11.40 and the Snowball is the only held object in 11.31
  (if (not (Variables|Default|GetHeld))
    (Variables|ProjectileMovement|SetRepStartMoving 1)
    ; Added in 31.00
    ; May want to move to the parent with the rest of this function. Should investigate.
    (LinkToActorComponent|SetAllowLinkingCheck (Variables|Default|GetFortLinkToActor) false) ; Prevent repetitive linking checks if the snowball is on a surface it can't link to (like the default island surface).\nThat causes OnLinkedActorDisconnected to fire over and over until picked up or destroyed. This stops that nonsense.\nAlso, the parent's OnProjectileStop quickly sets RepStartMoving to 0, which would continually fire the OnRep.
  ))

(event OnLinkedActorConnected(FortLinkToActor)
  (params (LinkedActor :type Actor%20Object%20Reference))
  (LinkToActorComponent|SetAllowLinkingCheck (Variables|Default|GetFortLinkToActor) true))

(event Custom|DoDamageEffectTotarget
  (params (HitActor :type Actor%20Object%20Reference))
  (bind (_asplayer_pawn_athena _bsuccess) (Utilities|Casting|CastToPlayerPawn_Athena HitActor))
  (bind (_asfort_player_pawn _bsuccess_2) (Utilities|Casting|CastToFortPlayerPawn HitActor))
  ;[1 1 1] Apply damage, frost feet, and impulse
  (bind _returnvalue (CallFunction|FilterbyLOS HitActor))
  (if _returnvalue
    (Curie|ExecuteWaterApplication (Variables|Default|GetLastHolder) (Variables|Default|GetLastHolder) HitActor 1e+02)
    (Utilities|Casting|CastToBuildingItemWeaponUpgradeActor HitActor
      (:CastFailed
        (if (not (and (Team|OnSameTeam HitActor (Variables|Default|GetLastHolder)) _bsuccess_2))
          (GameplayEffects|ApplyGameplayEffectSpecToTarget (Ability|GetAbilitySystemComponent (Variables|Default|GetLastHolder)) (select _bsuccess (GameplayEffects|MakeOutgoingSpec :self (Ability|GetAbilitySystemComponent (Variables|Default|GetLastHolder)) :GameplayEffectClass "/Game/Athena/Items/HeldObjects/Snowball/GE_HeldObject_Snowball_Damage.GE_HeldObject_Snowball_Damage_C" :Level 0.0) (GameplayEffects|MakeOutgoingSpec :self (Ability|GetAbilitySystemComponent (Variables|Default|GetLastHolder)) :GameplayEffectClass "/Game/Athena/Items/HeldObjects/Snowball/GE_HeldObject_Snowball_Damage_Env.GE_HeldObject_Snowball_Damage_Env_C" :Level 0.0)) (Ability|GetAbilitySystemComponent HitActor)))
        (bind _asfort_pawn (Utilities|Casting|CastToFortPawn HitActor))
        (Utilities|IsValid _asfort_pawn
          (:Is%20Valid
            (GameplayEffects|FortApplyGameplayEffectToTarget :self (Ability|GetFortAbilitySystemComponent (Variables|Default|GetLastHolder)) :GameplayEffectClass "/Game/Athena/Items/HeldObjects/Snowball/GE_Snowball_LowFriction_Athena.GE_Snowball_LowFriction_Athena_C" :TargetOfEffect (Ability|GetAbilitySystemComponent _asfort_pawn) :OptionalSourceObject (Variables|Default|GetLastHolder) :OptionalEffectCauser (Variables|Default|GetLastHolder) :Level 0.0)
            (Character|LaunchCharacter _asfort_pawn (* (Math|Vector|Normalize (- (Transformation|GetActorLocation _asfort_pawn) (Transformation|GetActorLocation))) 1.2e+03) true true))))) ; Don't destroy upgrade benches
  ))

(event Custom|OnDamageCompleted
  (params (WasCancelled :type Boolean))
  ;[1 1 1] Apply damage, frost feet, and impulse
  (CallFunction|EventSnowballDeath 0))

(event OnHeldObjectOwningPawnChanged(FortHeldObject)
  (bind _fortlinktoactor (Variables|Default|GetFortLinkToActor))
  ;[1 1 1] Handle Attachment
  (if (HeldObjects|IsHeld (Variables|Default|GetFortHeldObject))
    (LinkToActorComponent|SetAllowLinkingCheck _fortlinktoactor false) ; Without this, it still tries to link while held
    (LinkToActorComponent|UnbindandClearAll _fortlinktoactor)
    (bind _returnvalue (BuildingActor|PhysicsSimulationSupport|GetPhysicsObjectComponent self))
    (PhysicsObject|UnbindLinktoActorDelegates _returnvalue)
    ;;[0.8385 0.5766 0.1938 0.5] The bug only ever seems to occur on Client, so no need to check on Server (except to catch replays)
    (Networking|CanExecuteCosmeticEvents
      (:True
        (Utilities|FlowControl|WaitforBothExecutions true
          (:Then
            (Default|CheckforReattach)))))
    (else
      (LinkToActorComponent|SetAllowLinkingCheck (Variables|Default|GetFortLinkToActor) true)
      (bind _returnvalue_2 (BuildingActor|PhysicsSimulationSupport|GetPhysicsObjectComponent self))
      (PhysicsObject|BindLinktoActorDelegates _returnvalue_2))))

(event OnHeldObjectAttachAttempt(FortHeldObject)
  (params (bAttachSuccess :type Boolean))
  ;[1 1 1] Handle Attachment
  ;;[0.8385 0.5766 0.1938 0.5] The bug only ever seems to occur on Client, so no need to check on Server (except to catch replays)
  (Networking|CanExecuteCosmeticEvents
    (:True
      (Utilities|FlowControl|WaitforBothExecutions true
        (:Then
          (Default|CheckforReattach))))))

(fn BlueprintCanInteract
  (params (InteractingPawn :type Fort%20Pawn%20Object%20Reference) (InteractionBeingAttempted :type EInteractionBeingAttempted%20Enum :default "FirstInteraction") (InteractionType :type TInteractionType%20Enum :default "IT_NoInteraction"))
  (Utilities|IsValid InteractingPawn
    (:Is%20Valid
      (bind _returnvalue (HeldObjects|Interaction|CanInteract (Variables|Default|GetFortHeldObject) (Utilities|Casting|CastToFortPlayerPawn InteractingPawn)))
      (return (Math|Boolean|ANDBoolean _returnvalue (not (GameplayTags|HasMatchingGameplayTag InteractingPawn "(TagName=\"Gameplay.Status.HidingInProp\")")) (<= (Math|Vector|VectorLength (Variables|/Script/Engine.MovementComponent|GetVelocity (Variables|Default|GetFortProjectileMovement))) 0.0) (not (GameplayTags|HasAnyMatchingGameplayTags InteractingPawn (Variables|Interact|GetTC_BlockPickup))))))))

(fn BlueprintGetInteractionString
  (params (InteractingPawn :type Fort%20Pawn%20Object%20Reference) (InteractionBeingAttempted :type EInteractionBeingAttempted%20Enum :default "FirstInteraction"))
  (return "NSLOCTEXT(\"[150C97CD4F36522BE64BEF8CD3568975]\", \"2B81D955482154284F64ECA14197451F\", \"Pick Up\")"))

(fn CheckForReattach
  (bind _returnvalue (Actor|GetAttachParentActor))
  (bind _fortheldobject (Variables|Default|GetFortHeldObject))
  ;[1 1 1] Check the held object's velocity and location. At this point it should have no velocity, and the capsule's relative location should be the defined attachment offset (meaning it's properly attached). If the velocity or location is wrong, we reattach it.
  (bind _output_get (LocalVariables|SetHeldObjectOwner (HeldObjects|GetOwningPlayer (Variables|Default|GetFortHeldObject))))
  (Utilities|IsValid _output_get
    (:Is%20Valid
      (if (== _output_get _returnvalue)
        (if (not (Math|Vector|VectorIsZero (Variables|/Script/Engine.MovementComponent|GetVelocity (Variables|Default|GetFortProjectileMovement))))
          _
          (elif (Utilities|Operators|NotEqual(!=) (Variables|/Script/Engine.SceneComponent|GetRelativeLocation (Variables|Default|GetCapsule)) (Variables|/Script/FortniteGame.FortHeldObjectComponent|GetAttachmentOffset (Variables|Default|GetFortHeldObject)) 0.01)
            _)))
      (bind _returnvalue_2 (Transformation|AttachActorToComponent (Variables|/Script/Engine.Character|GetMesh (LocalVariables|GetHeldObjectOwner)) (Variables|/Script/FortniteGame.FortHeldObjectComponent|GetAttachmentSocketName (Variables|Default|GetFortHeldObject)) "SnapToTarget" "SnapToTarget" "SnapToTarget")) ; WeldSimulatedBodies is true to help ensure this works
      (if _returnvalue_2
        (Transformation|SetActorRelativeLocation (Variables|/Script/FortniteGame.FortHeldObjectComponent|GetAttachmentOffset _fortheldobject))
        (else
          (Fortnite|FortShippingLog _fortheldobject "B_HeldObject_Snowball::CheckForReattach() - AttachActorToComponent() failed" true))))
    (:Is%20Not%20Valid
      (Fortnite|FortShippingLog _returnvalue "B_HeldObject_Snowball::CheckForReattach() - Invalid Owning Player" true))))

(fn FilterByLOS
  (params (ActorToCheck :type Actor%20Object%20Reference))
  (bind _returnvalue (Fortnite|CheckLineOfSightToActorWithChannel (+ (Transformation|GetActorLocation self) (* (Variables|Default|GetImpactNormal) 5e+01)) ActorToCheck "ECC_GameTraceChannel5" self))
  (return _returnvalue))

(fn OnDamageCancelled
  (return false))

(fn ServerOnlyHandleAttachToVehicle
  (params (HitLoc :type Vector :default "0, 0, 0") (HitComponentToAttachTo :type Primitive%20Component%20Object%20Reference))
  (CallFunction|BlowUp 0))

(fn ServerOnlyHandlePlayerControllerOnUnPossess
  (params (PlayerPawn :type Pawn%20Object%20Reference))
  (CallFunction|BlowUp 0))

(fn ConstructionScript
  (|CallParentFunction))
